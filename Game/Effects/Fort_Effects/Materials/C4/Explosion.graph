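# Unlit translucent particle material for a toon explosion effect, in the C4 effects folder, usable on Cascade and Niagara sprites.
# A SubUV toon explosion texture is multiplied by particle color and alpha, so the emitter controls tint and fade.
# The color passes through the shared MF_FX_EmissiveAdjustments function to emissive.
# The clamped alpha feeds a depth fade, softening intersections with scene geometry.
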Material /Game/Effects/Fort_Effects/Materials/C4/Explosion
graph {
 properties {
  bUsedWithNiagaraSprites=true
  bUsedWithParticleSprites=true
  bUsedWithStaticMesh=false
  blendMode=BLEND_Translucent
  pixelDepthOffsetMode=PDOM_Legacy
  shadingModel=MSM_Unlit
  usedShadingModels=Unlit
 }
 nodes [
  n1 = ParticleColor()
  n2 = ParticleSubUV!(UVs, Tex, Apply View MipBias) {texture=/Game/Effects/Fort_Effects/Textures/Smoke/Main_explosion/T_SubUV_Explo_Toon}
  DepthFade(Opacity=Clamp(None=Multiply(A=@n1.A, B=@n2.A), Max=Constant() {r=1}))
  MaterialFunctionCall(Emissive=Multiply(A=@n1.RGB, B=@n2.RGB)) {materialFunction=&MF_FX_EmissiveAdjustments}
 ]
}
